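# Post-process AnimBP for the male medium GallonBag head cosmetic, driving its skeleton from the character mesh.
# IK-retargeted and copied parent poses are layered per bone, then pass through a Control Rig fed aiming (delayed release) and backbling flags.
# DefaultSlot blends onto neck and head bones, weighted by the front-end Frontend_Idle_Bypass curve.
# Outside the front end, a missing pawn owner switches retargeting to a posing profile with head/neck chain overrides.
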AnimBlueprint /BRCosmetics/Characters/Player/Male/Medium/Heads/M_MED_GallonBag_Head/Meshes/M_MED_GallonBag_Head_AnimBP_PP
target_skeleton=&/M_MED_GallonBag_Head_Skeleton
graphs [
 AnimGraph {nodes [
   n1 = Layered blend per bone(
    BasePose=Retarget Pose From Mesh!(CustomRetargetProfile:Retarget Profile Structure=Get Target Retarget Profile() -> (Target Retarget Profile:Retarget Profile Structure)) -> (Pose:Pose Link Structure) {
     is=title AnimGraphNode_RetargetPoseFromMesh
     iKRetargeterAsset=&/M_MED_GallonBag_Head_Retargeter
    }
    BlendPoses_0=Copy Pose From Mesh() {bUseAttachedParent=true}
    BlendWeights_0=Get CustomIdleIsPlayingFloat() -> (CustomIdleIsPlayingFloat:float)
   ) {layerSetup [{branchFilters [
       {boneName=spine_01}
       {boneName=pelvis}
       {boneName=deform_spine_01}
       {boneName=deform_spine_02}
       {boneName=deform_spine_03}
       {boneName=deform_spine_04}
       {boneName=deform_spine_05}
       {boneName=deform_neck_01}
       {boneName=deform_head}
       {boneName=deform_head_eyes_rot}
       {boneName=deform_eyes}
       {boneName=deform_head_main_rot}
       {boneName=deform_head_mid_rot}
       {boneName=deform_head_tip_rot}
       {boneName=deform_head_upper_rot}
      ]}]}
   n2 = Control Rig!(
    Source=Layered blend per bone(BasePose=@n1, BlendPoses_0=Copy Pose From Mesh() {bUseAttachedParent=true}) {layerSetup [{branchFilters [{boneName=spine_01}]}]}
    Alpha
    IsAds:bool=Get IsAds() -> (IsAds:bool)
    Wearing Backbling:bool=Get Is Wearing Backbling()
   ) -> (Pose) {
    bTransferInputCurves=false
    alphaBoolBlend {blendOption=Linear}
    controlRigAssetReference {blueprintRigClass=&/M_MED_GallonBag_Head_CtrlRig.M_MED_GallonBag_Head_CtrlRig_C}
   }
   n3 = Layered blend per bone(BasePose=@n2, BlendPoses_0=Slot 'DefaultSlot'(), BlendWeights_0=Get CustomIdleIsPlayingFloat()) {layerSetup [{branchFilters [
       {boneName=deform_neck_01}
       {boneName=deform_head}
       {boneName=deform_head_mid_rot}
       {boneName=deform_head_main_rot}
       {boneName=deform_head_upper_rot}
       {boneName=deform_head_tip_rot}
       {boneName=deform_head_eyes_rot}
       {boneName=deform_eyes}
      ]}]}
   Output Pose(Result=@n3)
  ]}
 EventGraph {nodes [
   Event BlueprintUpdateAnimation()
   > n1 = Branch(Condition=Get bIsInFrontEnd())
   > Set CustomIdleIsPlayingFloat(CustomIdleIsPlayingFloat:float=GetCurveValue(self=GetOwnerFrontEndAnimInstance(), CurveName=Frontend_Idle_Bypass)) -> (Output_Get:float)
   > Set Target Retarget Profile(Target Retarget Profile:Retarget Profile Structure=Get Default Retarget Profile() -> (Default Retarget Profile:Retarget Profile Structure)) -> (Output_Get:Retarget Profile Structure)
   n2 = Is Valid(exec=@n1.else, InputObject=TryGetPawnOwner())
   Set Enable Head Logic(execute=@n2.Is Not Valid, Enable Head Logic:double=0) -> (Output_Get:double)
   > Set Target Retarget Profile(Target Retarget Profile=Get Posing Retarget Profile() -> (Posing Retarget Profile:Retarget Profile Structure))
   Set Enable Head Logic(execute=@n2.Is Valid, Enable Head Logic=1)
   > Set Target Retarget Profile(Target Retarget Profile=Get Default Retarget Profile())
   n3 = Event OnCharacterAndPickaxeCustomizationLoaded!() -> (
    OutputDelegate
    then
    CustomBaseMesh:Skeletal Mesh Component Object Reference
    CustomParts:Array of Custom Character Part Object References
    CustomMeshParts:Array of Skeletal Mesh Component Budgeted Object References
   ) {is=title K2Node_Event}
   > Set Body(Body:Skeletal Mesh Component Budgeted Object Reference=Get Array Item with EFort Custom Part Type(Array=@n3.CustomMeshParts)) -> (Output_Get:Skeletal Mesh Component Budgeted Object Reference)
   > Set Is Wearing Backbling(Is Wearing Backbling=CheckForBackbling())
   > Set Enable Head Logic(Enable Head Logic=1)
   > Set Target Retarget Profile(Target Retarget Profile=Get Default Retarget Profile())
   Event OnBeginTargeting() {is=title K2Node_Event}
   > Set IsAds(IsAds:bool=true) -> (Output_Get:bool)
   Event OnEndTargeting() {is=title K2Node_Event}
   > Delay(Duration=1)
   > Set IsAds(IsAds=false)
  ]}
]
variables [
 Body {type=SkeletalMeshComponentBudgeted}
 CustomIdleIsPlayingFloat {type=number}
 Default Retarget Profile {
  type=RetargetProfile
  default {
   bApplyChainSettings=false
   bApplyGlobalSettings=false
   bApplyRootSettings=false
   bApplySourceRetargetPose=false
   bApplyTargetRetargetPose=false
   bForceAllIKOff=false
   sourceRetargetPoseName=None
   targetRetargetPoseName=None
   chainSettings {}
   globalSettings {
    bCopyBasePose=false
    bEnableFK=true
    bEnableIK=true
    bEnablePost=true
    bEnableRoot=true
    bWarping=false
    copyBasePoseRoot=None
    directionChain=None
    directionSource=Goals
    forwardDirection=Y
    sidewaysOffset=0
    sourceScaleFactor=1
    warpForwards=1
    warpSplay=1
   }
   retargetOpProfiles []
   rootSettings {
    affectIKHorizontal=1
    affectIKVertical=0
    blendToSource=0
    rotationAlpha=1
    scaleHorizontal=1
    scaleVertical=1
    translationAlpha=1
    blendToSourceWeights=<1 1 1>
    rotationOffset=<rot 0 0 0>
    translationOffset=<0 0 0>
   }
  }
 }
 Enable Head Logic {type=number, default=1}
 InverseIdlePlaying {type=number, default=1}
 Is Wearing Backbling {type=boolean}
 IsAds {type=boolean}
 Posing Retarget Profile {
  type=RetargetProfile
  default {
   bApplyChainSettings=true
   bApplyGlobalSettings=false
   bApplyRootSettings=false
   bApplySourceRetargetPose=false
   bApplyTargetRetargetPose=false
   bForceAllIKOff=false
   sourceRetargetPoseName=None
   targetRetargetPoseName=None
   chainSettings {
    head {
     fK {
      enableFK=true
      poleVectorMaintainOffset=false
      poleVectorMatching=0
      poleVectorOffset=0
      rotationAlpha=1
      rotationMode=Interpolated
      translationAlpha=1
      translationMode=None
     }
     iK {
      bAffectedByIKWarping=true
      blendToSource=0
      blendToSourceRotation=0
      blendToSourceTranslation=1
      enableIK=false
      extension=1
      scaleVertical=1
      blendToSourceWeights=<1 1 1>
      staticLocalOffset=<0 0 0>
      staticOffset=<0 0 0>
      staticRotationOffset=<rot 0 0 0>
     }
     speedPlanting {enableSpeedPlanting=false, speedCurveName=None, speedThreshold=15, unplantCriticalDamping=1, unplantStiffness=250}
    }
    neck {
     fK {
      enableFK=true
      poleVectorMaintainOffset=false
      poleVectorMatching=0
      poleVectorOffset=0
      rotationAlpha=1
      rotationMode=Interpolated
      translationAlpha=1
      translationMode=None
     }
     iK {
      bAffectedByIKWarping=true
      blendToSource=0
      blendToSourceRotation=0
      blendToSourceTranslation=1
      enableIK=false
      extension=1
      scaleVertical=1
      blendToSourceWeights=<1 1 1>
      staticLocalOffset=<0 0 0>
      staticOffset=<0 0 0>
      staticRotationOffset=<rot 0 0 0>
     }
     speedPlanting {enableSpeedPlanting=false, speedCurveName=None, speedThreshold=15, unplantCriticalDamping=1, unplantStiffness=250}
    }
   }
   globalSettings {
    bCopyBasePose=false
    bEnableFK=true
    bEnableIK=true
    bEnablePost=true
    bEnableRoot=true
    bWarping=false
    copyBasePoseRoot=None
    directionChain=None
    directionSource=Goals
    forwardDirection=Y
    sidewaysOffset=0
    sourceScaleFactor=1
    warpForwards=1
    warpSplay=1
   }
   retargetOpProfiles []
   rootSettings {
    affectIKHorizontal=1
    affectIKVertical=0
    blendToSource=0
    rotationAlpha=1
    scaleHorizontal=1
    scaleVertical=1
    translationAlpha=1
    blendToSourceWeights=<1 1 1>
    rotationOffset=<rot 0 0 0>
    translationOffset=<0 0 0>
   }
  }
 }
 Target Retarget Profile {
  type=RetargetProfile
  default {
   bApplyChainSettings=false
   bApplyGlobalSettings=false
   bApplyRootSettings=false
   bApplySourceRetargetPose=false
   bApplyTargetRetargetPose=false
   bForceAllIKOff=false
   sourceRetargetPoseName=None
   targetRetargetPoseName=None
   chainSettings {}
   globalSettings {
    bCopyBasePose=false
    bEnableFK=true
    bEnableIK=true
    bEnablePost=true
    bEnableRoot=true
    bWarping=false
    copyBasePoseRoot=None
    directionChain=None
    directionSource=Goals
    forwardDirection=Y
    sidewaysOffset=0
    sourceScaleFactor=1
    warpForwards=1
    warpSplay=1
   }
   retargetOpProfiles []
   rootSettings {
    affectIKHorizontal=1
    affectIKVertical=0
    blendToSource=0
    rotationAlpha=1
    scaleHorizontal=1
    scaleVertical=1
    translationAlpha=1
    blendToSourceWeights=<1 1 1>
    rotationOffset=<rot 0 0 0>
    translationOffset=<0 0 0>
   }
  }
 }
]
